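# Light function material for the Fortress skybox that projects an animated, spinning grayscale pattern onto lighting.
# A rotating spiral texture multiplies a panning fort texture, an optional third panning layer modifies it, and contrast, inversion and shadow lightening shape the result.
# Camera distance switches UV tiling to a much larger pattern far away, then drives a partial brighten and a fade to dark.
# A static switch can replace time with a parameter, and low quality outputs a flat constant.
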Material /Game/Environments/World/LightFunctions/M_Fortress_Skybox_LF_Spinning
graph {
 properties {materialDomain=MD_LightFunction, pixelDepthOffsetMode=PDOM_Legacy, shadingModel=MSM_Unlit, usedShadingModels=Unlit}
 nodes [
  n1 = Distance(A=CameraPositionWS(), B=WorldPosition().XYZ)
  n2 = Constant() {r=38000}
  n3 = Constant() {r=0.013}
  n4 = Multiply(A=TextureCoordinate() {uTiling=0.1, vTiling=0.1}, B=@n3)
  n5 = TextureCoordinate() {uTiling=0.2, vTiling=0.2}
  n6 = If(A=@n1, B=@n2, "A > B"=@n4, "A == B"=@n5, "A < B"=@n5)
  n7 = StaticSwitchParameter(True=Time(), False=ScalarParameter() {defaultValue=1, parameterName=Time}) {defaultValue=true, parameterName=Animate?}
  n8 = If(A=@n1, B=@n2, "A > B"=@n4, "A == B"=@n5, "A < B"=@n5)
  n9 = Multiply(
   A=TextureSample(UVs=Rotator(Coordinate=@n6, Time=@n7) {speed=0.003}) {samplerType=SAMPLERTYPE_Grayscale, texture=&/T_spiralLF}.R
   B=TextureSample(UVs=Panner(Coordinate=@n8, Time=@n7) {speedY=0.01}) {samplerType=SAMPLERTYPE_Grayscale, texture=&/T_FORT_LF}.R
  )
  n10 = TextureCoordinate() {uTiling=0.4, vTiling=0.4}
  n11 = If(A=@n1, B=@n2, "A > B"=Multiply(A=TextureCoordinate() {uTiling=0.25, vTiling=0.25}, B=@n3), "A == B"=@n10, "A < B"=@n10)
  n12 = ScalarParameter() {defaultValue=1, parameterName=3rd texture scale}
  n13 = TextureSample(UVs=Panner(Coordinate=Multiply(A=@n11, B=@n12), Time=@n7) {speedY=0.01}) {samplerType=SAMPLERTYPE_Grayscale, texture=&T_FORT_LF}
  n14 = Multiply(
   A=OneMinus(None=@n9)
   B=LinearInterpolate(
    A=ScalarParameter() {parameterName=Lerp Sub A}
    B=ScalarParameter() {defaultValue=0.1, parameterName=Lerp Sub B}
    Alpha=Multiply(A=@n13.R, B=@n13.R)
   ) {constB=0.1}
  )
  n15 = Subtract(A=@n9, B=@n14)
  n16 = LinearInterpolate(
   A=ScalarParameter() {defaultValue=-0.5, parameterName=ContrastLow}
   B=ScalarParameter() {defaultValue=8, parameterName=ContrastHigh}
   Alpha=StaticSwitchParameter(True=Clamp(None=@n15), False=Multiply(A=@n9, B=@n14)) {defaultValue=true, parameterName=Use Third Layer}
  ) {constA=-0.5, constB=8}
  n17 = Add(A=OneMinus(None=Clamp(None=@n16)), B=ScalarParameter() {defaultValue=0.125, parameterName=LightenShadow}) {constB=0.2}
  n18 = ConstantBiasScale(None=Multiply(A=Clamp(None=@n17) {minDefault=0.05}, B=Constant() {r=2})) {bias=0.5, scale=0.6667}
  n19 = Multiply(A=@n18, B=@n18)
  n20 = CameraPositionWS()
  n21 = WorldPosition()
  n22 = Divide(A=Subtract(A=Distance(A=@n20, B=@n21.XYZ), B=Constant() {r=250000}), B=ScalarParameter() {defaultValue=18000, parameterName=Bright Distance})
  n23 = Subtract(A=Distance(A=CameraPositionWS(), B=WorldPosition().XYZ), B=Constant() {r=16000})
  n24 = ScalarParameter() {defaultValue=1800000, parameterName=Dark Dist}
  n25 = LinearInterpolate(
   A=LinearInterpolate(A=@n19, B=LinearInterpolate(B=@n19) {constA=1, constAlpha=0.35, constB=0}, Alpha=Clamp(None=@n22))
   Alpha=Clamp(None=Divide(A=@n23, B=@n24))
  ) {constB=0}
  QualitySwitch(Default=@n25, Low=Constant() {r=0.6}, High=@n25)
 ]
}
