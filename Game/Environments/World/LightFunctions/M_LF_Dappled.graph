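# Light function material, compatible with the light function atlas, that casts animated dappled light.
# A slowly rotating spiral texture multiplies a panning Voronoi cloud texture, then a squared, counter-panning Voronoi sample modulates the mask.
# The inverted result is shaped by FinalPower and FinalBrightness, with LightenShadow lifting the dark areas.
# The Animate? static switch swaps live time for a Time parameter to freeze or manually drive the motion.
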Material /Game/Environments/World/LightFunctions/M_LF_Dappled
graph {
 properties {
  bForceCompatibleWithLightFunctionAtlas=true
  materialDomain=MD_LightFunction
  pixelDepthOffsetMode=PDOM_Legacy
  shadingModel=MSM_Unlit
  usedShadingModels=Unlit
 }
 nodes [
  Constant()
  n1 = TextureCoordinate() {uTiling=0.2, vTiling=0.2}
  n2 = StaticSwitchParameter(True=Time(), False=ScalarParameter() {defaultValue=1, parameterName=Time}) {defaultValue=true, parameterName=Animate?}
  n3 = Multiply(
   A=TextureSample(UVs=Rotator(Coordinate=@n1, Time=@n2) {speed=0.003}) {samplerType=SAMPLERTYPE_Grayscale, texture=&/T_spiralLF}.R
   B=TextureSample(UVs=Panner(Coordinate=@n1, Time=@n2) {speedY=0.01}) {samplerType=SAMPLERTYPE_LinearGrayscale, texture=/Game/Packages/Fortress_Sky/SkyDome/Master/T_SkyDome_Cloud_Voroni}.R
  )
  n4 = Constant() {r=-0.1497}
  n5 = Constant()
  n6 = TextureSample(UVs=Panner(Coordinate=TextureCoordinate() {uTiling=0.5, vTiling=0.5}, Time=@n2) {speedY=-0.01}) {samplerType=SAMPLERTYPE_LinearGrayscale, texture=&T_SkyDome_Cloud_Voroni}
  n7 = Multiply(A=@n6.R, B=@n6.R)
  n8 = Subtract(A=@n3, B=Multiply(A=OneMinus(None=@n3), B=LinearInterpolate(A=@n4, B=@n5, Alpha=@n7) {constB=0.1}))
  n9 = Power(Base=OneMinus(None=Clamp(None=@n8)), Exp=ScalarParameter() {defaultValue=7, parameterName=FinalPower})
  n10 = ScalarParameter() {defaultValue=4, parameterName=FinalBrightness}
  LinearInterpolate(A=ScalarParameter() {defaultValue=0.35, parameterName=LightenShadow}, Alpha=Clamp(None=Multiply(A=@n9, B=@n10)) {clampMode=CMODE_ClampMax})
 ]
}
