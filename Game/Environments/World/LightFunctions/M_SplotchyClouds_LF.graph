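# Light function material that produces splotchy, drifting cloud patterns to modulate lighting.
# A panning swirl flowmap distorts two samples of a storm cloud detail texture, each tiled separately and scrolled in opposite directions.
# The samples are multiplied, shaped by an inverted power curve, scaled and clamped to an adjustable minimum.
# Camera-distance tiling and an alternate powered multiply are also present but do not feed that chain.
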Material /Game/Environments/World/LightFunctions/M_SplotchyClouds_LF
graph {
 properties {materialDomain=MD_LightFunction, pixelDepthOffsetMode=PDOM_Legacy, shadingModel=MSM_Unlit, usedShadingModels=Unlit}
 nodes [
  n1 = Multiply(A=ScalarParameter() {defaultValue=0.15, group=Flowmap, parameterName=FlowmapTile}, B=TextureCoordinate())
  n2 = ComponentMask(None=@n1) {r=true}
  n3 = Time()
  n4 = Multiply(A=ScalarParameter() {defaultValue=0.015, group=Flowmap, parameterName=FlowmapPanSpeed}, B=@n3)
  n5 = ComponentMask(None=@n1) {g=true}
  n6 = TextureSample(UVs=AppendVector(A=Add(A=@n2, B=@n4), B=Add(A=@n5, B=@n4))) {samplerType=SAMPLERTYPE_LinearColor, texture=/Game/VisualThreatManager/StormVisuals/LargeClouds/T_Swirl_Flowmap}
  n7 = Multiply(
   A=Subtract(A=ComponentMask(None=@n6.RGB) {g=true, r=true}) {constB=0.5}
   B=ScalarParameter() {defaultValue=0.3, group=Flowmap, parameterName=FlowmapAmount}
  ) {constB=0.5}
  n8 = Multiply(A=@n3, B=ScalarParameter() {defaultValue=0.02, group=Mask, parameterName=TexturePanSpeed})
  n9 = VectorParameter() {parameterName=FlowTileV, defaultValue=<rgba 2 1 0.18 0>}
  n10 = Multiply(A=Multiply(A=ComponentMask(None=@n9.RGB) {g=true, r=true}, B=@n9.B), B=TextureCoordinate())
  n11 = Add(A=Multiply(A=@n8, B=Constant() {r=-1}), B=ComponentMask(None=@n10) {r=true})
  n12 = Add(A=ComponentMask(None=@n10) {g=true}, B=@n8)
  n13 = TextureSample(UVs=Add(A=@n7, B=AppendVector(A=@n11, B=@n12))) {desc=Texture A, samplerType=SAMPLERTYPE_LinearGrayscale, texture=/Game/Global/Textures/Default/Tileable_Patterns/T_StormCloudDetail01_M}
  n14 = VectorParameter() {parameterName=FlowTileV_1, defaultValue=<rgba 2 1 0.15 0>}
  n15 = Multiply(A=Multiply(A=ComponentMask(None=@n14.RGB) {g=true, r=true}, B=@n14.B), B=TextureCoordinate())
  n16 = Add(A=ScalarParameter() {defaultValue=0.5, group=Mask, parameterName=TextureBOffset}, B=Add(A=@n8, B=ComponentMask(None=@n15) {r=true}))
  n17 = Add(A=ComponentMask(None=@n15) {g=true}, B=Multiply(A=@n8, B=Constant() {r=-1}))
  n18 = TextureSample(UVs=Add(A=@n7, B=AppendVector(A=@n16, B=@n17))) {desc=Texture B, samplerType=SAMPLERTYPE_LinearGrayscale, texture=&T_StormCloudDetail01_M}
  n19 = Power(Base=OneMinus(None=Multiply(A=@n13.R, B=@n18.R)), Exp=ScalarParameter() {defaultValue=1, group=Mask, parameterName=TexturePoweredMultiply_1}) {constExponent=3}
  Clamp(
   None=Multiply(A=OneMinus(None=@n19), B=ScalarParameter() {defaultValue=1, parameterName=FinalMultiplier})
   Min=ScalarParameter() {defaultValue=0.1, parameterName=FinalMinClamp}
  )
  Constant() {r=20}
  Constant() {r=0.3}
  Constant() {r=38000}
  n20 = ScalarParameter() {defaultValue=4, group=Mask, parameterName=TexturePoweredMultiply}
  Multiply(A=Power(Base=@n13.R, Exp=@n20) {constExponent=3}, B=Power(Base=@n18.R, Exp=@n20) {constExponent=3})
  n21 = ScalarParameter() {defaultValue=25000, parameterName=FarDistance}
  n22 = TextureCoordinate() {uTiling=0.2, vTiling=0.2}
  n23 = If(A=Distance(A=CameraPositionWS(), B=WorldPosition().XYZ), B=@n21, "A > B"=@n21, "A == B"=@n22, "A < B"=@n22)
  n24 = ScalarParameter() {defaultValue=1, group=Mask, parameterName=TextureTile}
  Multiply(A=@n23, B=@n24)
  Multiply(A=@n23, B=@n24)
  StaticSwitchParameter(True=@n23, False=TextureCoordinate()) {defaultValue=true, parameterName=UseDistanceCheck}
 ]
}
